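# Material function that produces an animated, masked reactive fire color for the BurnLeer pickaxe.
# Panned noise over an input UV, defaulting to a triplanar front projection, forms a distortion that appears to offset a view-angle falloff.
# That value appears to index a gradient texture, giving fire color without authored flame art.
# The result is scaled by an intensity parameter, a Fresnel-driven orange tint and the ReactiveFireMask input.
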MaterialFunction /BRCosmetics/Weapons/FORT_Melee/Pickaxe_BurnLeer/Materials/MF_Picaxe_BurnLeer_ReactiveFire
graph {nodes [
  n1 = TextureSample() {
   samplerSource=SSM_Wrap_WorldGroupSettings
   samplerType=SAMPLERTYPE_Masks
   texture=/BRCosmetics/Characters/Player/Female/Medium/Bodies/F_MED_BurnLeer/Textures/T_BurnLeer_Body_FX3
  }
  n2 = FunctionInput(Preview=@n1.R) {bUsePreviewValueAsDefault=true, inputName=ReactiveFireMask, inputType=FunctionInput_Scalar}
  n3 = Constant() {r=0.1}
  n4 = Multiply(
   A=Multiply(A=NamedRerouteUsage(), B=ScalarParameter() {defaultValue=0.75, group=BurnLeer, parameterName=ReactiveFirestIntensity})
   B=LinearInterpolate(A=Constant4Vector() {constant=<rgba 1 0.1134 0.0015 1>}.RGB, Alpha=Fresnel(ExponentIn=@n3) {exponent=1})
  )
  FunctionOutput(None=NamedRerouteDeclaration(None=Multiply(A=@n4, B=@n2)) {name=ReactiveFire, nodeColor=<rgba 0.8281 1 0 1>})
  n5 = TextureSample(UVs=MaterialFunctionCall(X=NamedRerouteUsage(), Y=Constant() {r=0.5}) {materialFunction=&MakeFloat2}.Result) {texture=/BRCosmetics/Characters/Player/Male/Medium/Bodies/M_MED_TurboFueled/FX/T_TurboFueled_Gradient}
  n6 = Constant() {r=1.3}
  NamedRerouteDeclaration(None=Multiply(A=Power(Base=@n5.RGB, Exp=@n6), B=Constant() {r=1})) {name=EmissiveOUT_1, nodeColor=<rgba 1 0.0938 0 1>}
  n7 = Power(Base=NamedRerouteUsage(), Exp=Constant() {r=0.5})
  NamedRerouteDeclaration(None=Multiply(A=ConstantBiasScale(None=@n7) {bias=-0.5, scale=2}, B=Constant() {r=0.4})) {name=Distortion, nodeColor=<rgba 1 0.7031 0 1>}
  n8 = Saturate(None=DotProduct(A=CameraVectorWS(), B=PixelNormalWS()))
  n9 = Constant() {r=1.2258}
  n10 = Add(A=NamedRerouteUsage(), B=NamedRerouteDeclaration(None=Power(Base=@n8, Exp=@n9)) {name=FresenlValue, nodeColor=<rgba 0.7344 0 1 1>})
  NamedRerouteDeclaration(None=LinearInterpolate(Alpha=Saturate(None=@n10)) {constB=0.99}) {name=GradientUVX, nodeColor=<rgba 0 1 0.0625 1>}
  n11 = NamedRerouteUsage()
  n12 = Constant() {r=1.5}
  n13 = TextureSampleParameter2D(UVs=Panner(Coordinate=Multiply(A=@n11, B=@n12) {constB=4}, Speed=Constant4Vector() {constant=<rgba 0 0.15 0 0>}.RGB) {speedY=0.3}) {
   group=BurnLeer
   parameterName=T_ReactiveFire
   samplerSource=SSM_Wrap_WorldGroupSettings
   samplerType=SAMPLERTYPE_LinearColor
   texture=/Game/Global/Textures/Default/Noise/T_Curie_NoisePack
  }
  NamedRerouteDeclaration(None=Saturate(None=ComponentMask(None=@n13.RGB) {g=true, r=true})) {name=DistortionNoiseSample, nodeColor=<rgba 1 0 0.5391 1>}
  n14 = Constant()
  n15 = Constant()
  n16 = Constant4Vector() {constant=<rgba 0 0 0 1>}
  n17 = Constant() {r=30}
  NamedRerouteDeclaration(None=FunctionInput(Preview=MaterialFunctionCall!(TriplanarUV_UseClamp?=@n14, TriplanarUV_FlipUV=@n15, TriplanarUV_CenterPosition=@n16.RGB, TriplanarUV_UVScale=@n17) {materialFunction=/BRCosmetics/Characters/Player/Male/Medium/Bodies/M_MED_BirchPalace/Skins/Gas/Materials/MF_TriplanarUV}.UV_FrontProjection) {bUsePreviewValueAsDefault=true, inputName=InUV, inputType=FunctionInput_Vector2}) {name=UV Front, nodeColor=<rgba 0.4062 0 1 1>}
 ]}
