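# Parent material for the BurnLeer pickaxe, adding animated fire to the standard 2023 character material between its PreFX and PostFX functions.
# Lerps layer fire effects into base color and emissive, using a reactive fire function and a triplanar front-projected texture distorted by flowing smoke.
# Fire world position offset pushes vertices along normals with panned noise, faded by two inverted sphere masks in pre-skinned local space.
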Material /BRCosmetics/Weapons/FORT_Melee/Pickaxe_BurnLeer/Materials/M_Pickaxe_BurnLeer_Parent
graph {
 properties {
  bAutomaticallySetUsageInEditor=false
  bUseMaterialAttributes=true
  bUsedWithMeshDeformer=true
  bUsedWithMorphTargets=true
  bUsedWithSkeletalMesh=true
  pixelDepthOffsetMode=PDOM_Legacy
  shadingModel=MSM_FromMaterialExpression
  usedShadingModels=Default Lit | Subsurface | Clear Coat | Subsurface Profile | Cloth
 }
 nodes [
  n1 = GetMaterialAttributes(None=NamedRerouteUsage()) {attributeGetTypes [#1, #2]}
  n2 = NamedRerouteUsage()
  n3 = Add(A=NamedRerouteUsage(), B=Multiply(A=NamedRerouteUsage(), B=Constant()) {constB=0.2})
  n4 = NamedRerouteUsage()
  n5 = NamedRerouteUsage()
  n6 = Multiply(A=NamedRerouteUsage(), B=Constant() {r=0.2})
  n7 = NamedRerouteUsage()
  n8 = MaterialFunctionCall() {materialFunction=&MF_CharacterParent_2023_PreFX}
  MaterialFunctionCall(
   Material=SetMaterialAttributes!(
    MaterialAttributes=@n1.MaterialAttributes
    World Position Offset=NamedRerouteUsage()
    Emissive Color=LinearInterpolate(A=@n2, B=@n3, Alpha=@n4)
    Base Color=LinearInterpolate(A=@n5, B=@n6, Alpha=@n7)
   ) {attributeSetTypes [#3, #1, #2]}
   M=@n8.M
   Pre Skinned Local Position=@n8.M
   Distance Blend=@n8.M
   Fresnel=@n8.M
  ) {materialFunction=&MF_CharacterParent_2023_PostFX}
  n9 = Multiply(A=NamedRerouteUsage(), B=Constant() {r=0.5}) {constB=0.1}
  n10 = Multiply(A=NamedRerouteUsage(), B=Constant() {r=1.9357}) {constB=10}
  n11 = Constant() {r=0.001}
  n12 = Constant() {r=0.035}
  n13 = NamedRerouteUsage()
  n14 = Constant() {r=1}
  n15 = ComponentMask(None=NamedRerouteDeclaration(None=MaterialFunctionCall!(InUV=@n13, ReactiveFireMask=@n14) {materialFunction=&/MF_Picaxe_BurnLeer_ReactiveFire}.Result) {name=BurnLeerReactiveFire, nodeColor=<rgba 1 0.7031 0 1>}) {r=true}
  n16 = Multiply(A=@n15, B=Constant())
  n17 = Multiply(A=@n15, B=Constant() {r=0.12})
  n18 = Add(
   A=Multiply(
    A=MaterialFunctionCall!(In_FireSmoke_TAMask, In_FireSmoke_UV=@n10) {materialFunction=/BRCosmetics/Gadgets/Assets/VinderTech_GliderChute/Glider_RampCanal/Materials/MFW_FlowingSmoke}.Result
    B=AppendVector(A=@n11, B=@n12)
   )
   B=Add(A=NamedRerouteUsage(), B=AppendVector(A=@n16, B=@n17))
  )
  NamedRerouteDeclaration(None=Max(A=Power(Base=@n9), B=TextureSample(UVs=@n18) {texture=/BRCosmetics/Weapons/FORT_Melee/Pickaxe_BurnLeer/Textures/T_BurnLeer_Pickaxe_E2}.RGB)) {name=ProjectionForntFire, nodeColor=<rgba 1 0 0.9609 1>}
  n19 = MaterialFunctionCall!(
   TriplanarUV_UseClamp?=Constant()
   TriplanarUV_FlipUV=Constant()
   TriplanarUV_CenterPosition=Constant4Vector() {constant=<rgba 0.9523 0 15.2308 1>}.RGB
   TriplanarUV_UVScale=Constant() {r=58.5987}
  ) {materialFunction=/BRCosmetics/Characters/Player/Male/Medium/Bodies/M_MED_BirchPalace/Skins/Gas/Materials/MF_TriplanarUV}
  n20 = OneMinus(None=ComponentMask(None=@n19.UV_FrontProjection) {r=true})
  n21 = ComponentMask(None=@n19.UV_FrontProjection) {g=true}
  NamedRerouteDeclaration(None=MaterialFunctionCall(UVs=AppendVector(A=@n20, B=@n21), "Rotation Angle (0-1)"=Constant() {r=-0.012}) {materialFunction=&CustomRotator}.Rotated Values) {name=ProjectionForntUV, nodeColor=<rgba 0 0.0859 1 1>}
  n22 = LocalPosition() {localOrigin=InstancePreSkinning}
  n23 = Panner(Coordinate=Multiply(A=ComponentMask(None=@n22.XYZ) {b=true, r=true}, B=Constant() {r=0.048})) {speedY=-0.2}
  n24 = Multiply(
   A=Multiply(
    A=TextureSample(UVs=@n23) {texture=/Game/Global/Textures/Default/Noise/TankMoteNoise}.R
    B=ScalarParameter() {defaultValue=1.1, group=Pickaxe_BurnLeer, parameterName=WPOIntensity}
   )
   B=NamedRerouteUsage()
  )
  n25 = LocalPosition() {localOrigin=InstancePreSkinning}
  n26 = Constant4Vector() {constant=<rgba 0 0 45.0653 1>}
  n27 = Constant() {r=31.7338}
  n28 = Constant() {r=0.15}
  n29 = LocalPosition() {localOrigin=InstancePreSkinning}
  n30 = Constant4Vector() {constant=<rgba 0 0 -9.8556 1>}
  n31 = Constant() {r=23.5786}
  n32 = Constant() {r=0.85}
  n33 = Multiply(
   A=OneMinus(None=SphereMask(A=@n25.XYZ, B=@n26.RGB, Radius=@n27, Hardness=@n28))
   B=OneMinus(None=SphereMask(A=@n29.XYZ, B=@n30.RGB, Radius=@n31, Hardness=@n32))
  )
  NamedRerouteDeclaration(None=Multiply(A=VertexNormalWS(), B=Multiply(A=@n24, B=@n33))) {name=FireWPO, nodeColor=<rgba 0 1 0.9766 1>}
  NamedRerouteDeclaration(None=@n8.Result) {name=IN_MA, nodeColor=<rgba 1 0.1641 0 1>}
  NamedRerouteDeclaration(None=TextureSample() {samplerType=SAMPLERTYPE_Masks, texture=/BRCosmetics/Weapons/FORT_Melee/Pickaxe_BurnLeer/Textures/T_BusrnLeer_Axe_TA_Mask}.R) {name=BladeMask, nodeColor=<rgba 0 0.1562 1 1>}
  NamedRerouteDeclaration(None=@n1.Emissive Color) {name=IN_EmissiveColor, nodeColor=<rgba 0.4766 1 0 1>}
  NamedRerouteDeclaration(None=@n1.Base Color) {name=IN_BaseColor, nodeColor=<rgba 1 0 0.7969 1>}
 ]
}
